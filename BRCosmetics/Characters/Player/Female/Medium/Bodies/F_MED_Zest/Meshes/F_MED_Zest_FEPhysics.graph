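# Physics asset for the F_MED_Zest female character body, providing secondary motion for a hoodie with tied sleeves and two grenades.
# Three hoodie-back chains, a sleeve-tie joint branching into both sleeves, and both grenades hang from spine_01 on swing-limited, tight-twist constraints.
# A free-angle constraint cross-links the tips of hoodie chains 1 and 2.
# Pelvis, spine and thigh capsules form the body, joined by free-angle constraints.
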PhysicsAsset /BRCosmetics/Characters/Player/Female/Medium/Bodies/F_MED_Zest/Meshes/F_MED_Zest_FEPhysics
bodies {
 dyn_L_sleeve_1_jnt [{shapeType=Capsule, length=2.25, radius=2.5, shapeName=dyn_L_sleeve_1_jnt_capsule}]
 dyn_L_sleeve_2_jnt [{shapeType=Capsule, length=3, radius=3, shapeName=dyn_L_sleeve_2_jnt_capsule}]
 dyn_L_sleeve_3_jnt [{shapeType=Capsule, length=3, radius=3, shapeName=dyn_L_sleeve_3_jnt_capsule}]
 dyn_R_sleeve_1_jnt [{shapeType=Capsule, length=2.75, radius=2.25, shapeName=dyn_R_sleeve_1_jnt_capsule, center=<2.2218 0.7021 0>}]
 dyn_R_sleeve_2_jnt [{shapeType=Capsule, length=3, radius=3, shapeName=dyn_R_sleeve_2_jnt_capsule, center=<2.2218 0 0>}]
 dyn_grenade1_jnt [{shapeType=Capsule, length=4, radius=4, shapeName=dyn_grenade1_jnt_capsule, center=<4.0562 0 0>, rotation=<rot 90 0.633 0.6329>}]
 dyn_grenade2_jnt [{shapeType=Capsule, length=4, radius=4, shapeName=dyn_grenade2_jnt_capsule, center=<4.0562 0.2185 0>, rotation=<rot 90 -1.2658 -1.2657>}]
 dyn_hoodie_back_1_1_jnt [{
   shapeType=Capsule
   length=10.75
   radius=4
   shapeName=dyn_hoodie_back_1_1_jnt_capsule
   center=<3.7133 -0.0003 -2.8001>
   rotation=<rot 0 -0.0001 90.0004>
  }]
 dyn_hoodie_back_1_2_jnt [{
   shapeType=Capsule
   length=10.75
   radius=3.25
   shapeName=dyn_hoodie_back_1_2_jnt_capsule
   center=<3.7133 -0.0001 -2.7473>
   rotation=<rot 0 0 90.0004>
  }]
 dyn_hoodie_back_1_3_jnt [{
   shapeType=Capsule
   length=11
   radius=3.25
   shapeName=dyn_hoodie_back_1_3_jnt_capsule
   center=<3.7134 -0.0002 -1.2442>
   rotation=<rot 0 0 90.0004>
  }]
 dyn_hoodie_back_2_1_jnt [{shapeType=Capsule, length=7.5, radius=3.5, shapeName=dyn_hoodie_back_2_1_jnt_capsule, center=<3.3196 -4.3837 0>}]
 dyn_hoodie_back_2_2_jnt [{shapeType=Capsule, length=8.25, radius=3.5, shapeName=dyn_hoodie_back_2_2_jnt_capsule, center=<3.3196 -3.7426 0>}]
 dyn_hoodie_back_2_3_jnt [{shapeType=Capsule, length=8.5, radius=4, shapeName=dyn_hoodie_back_2_3_jnt_capsule, center=<3.3196 -3.0263 0>}]
 dyn_hoodie_back_3_1_jnt [{shapeType=Capsule, length=6.5, radius=1.75, shapeName=dyn_hoodie_back_3_1_jnt_capsule, center=<2.709 -1.1539 1>, rotation=<rot 90 0 0>}]
 dyn_hoodie_back_3_2_jnt [{shapeType=Capsule, length=8, radius=2.5, shapeName=dyn_hoodie_back_3_2_jnt_capsule, center=<1.8812 -1.1106 1>, rotation=<rot 90 0 0>}]
 dyn_sleeves_tie_jnt [{shapeType=Capsule, length=3, radius=3, shapeName=dyn_sleeves_tie_jnt_capsule, center=<0 0 1.086>}]
 pelvis [{shapeType=Capsule, length=8.25, radius=10.25, shapeName=pelvis_capsule, center=<-4.1886 -2.3933 1.435>}]
 spine_01 [{shapeType=Capsule, length=3, radius=3, shapeName=spine_01_capsule}]
 spine_02 [{shapeType=Capsule, length=12.25, radius=7.5, shapeName=spine_02_capsule, center=<3.4974 2.6821 0>}]
 thigh_l [{
   shapeType=Capsule
   length=32.7634
   radius=7.5186
   shapeName=thigh_l_capsule
   center=<-17.6633 0.023 0.0002>
   rotation=<rot 0 -93.702 -88.5352>
  }]
 thigh_r [{
   shapeType=Capsule
   length=23.5134
   radius=7.5186
   shapeName=thigh_r_capsule
   center=<15.3377 0.9343 -0.6449>
   rotation=<rot 0 -93.702 -88.5352>
  }]
}
constraints [
 {bone1Name=dyn_hoodie_back_1_2_jnt, bone2Name=dyn_hoodie_back_1_1_jnt, swing1LimitDegrees=20, swing2LimitDegrees=45, twistLimitDegrees=0.5}
 {bone1Name=dyn_hoodie_back_1_3_jnt, bone2Name=dyn_hoodie_back_1_2_jnt, swing1LimitDegrees=20, swing2LimitDegrees=45, twistLimitDegrees=0.5}
 {bone1Name=dyn_hoodie_back_2_2_jnt, bone2Name=dyn_hoodie_back_2_1_jnt, swing1LimitDegrees=45, swing2LimitDegrees=30, twistLimitDegrees=2}
 {bone1Name=dyn_hoodie_back_2_3_jnt, bone2Name=dyn_hoodie_back_2_2_jnt, swing1LimitDegrees=45, swing2LimitDegrees=30, twistLimitDegrees=2}
 {bone1Name=dyn_hoodie_back_3_2_jnt, bone2Name=dyn_hoodie_back_3_1_jnt, swing1LimitDegrees=45, swing2LimitDegrees=25, twistLimitDegrees=2}
 {bone1Name=dyn_L_sleeve_1_jnt, bone2Name=dyn_sleeves_tie_jnt, swing1LimitDegrees=45, swing2LimitDegrees=45, twistLimitDegrees=2}
 {bone1Name=dyn_L_sleeve_2_jnt, bone2Name=dyn_L_sleeve_1_jnt, swing1LimitDegrees=45, swing2LimitDegrees=45, twistLimitDegrees=2}
 {bone1Name=dyn_L_sleeve_3_jnt, bone2Name=dyn_L_sleeve_2_jnt, swing1LimitDegrees=45, swing2LimitDegrees=45, twistLimitDegrees=2}
 {bone1Name=dyn_R_sleeve_1_jnt, bone2Name=dyn_sleeves_tie_jnt, swing1LimitDegrees=45, swing2LimitDegrees=45, twistLimitDegrees=2}
 {bone1Name=dyn_R_sleeve_2_jnt, bone2Name=dyn_R_sleeve_1_jnt, swing1LimitDegrees=45, swing2LimitDegrees=45, twistLimitDegrees=2}
 {bone1Name=dyn_hoodie_back_1_1_jnt, bone2Name=spine_01, swing1LimitDegrees=45, swing2LimitDegrees=45, twistLimitDegrees=2}
 {bone1Name=dyn_hoodie_back_2_1_jnt, bone2Name=spine_01, swing1LimitDegrees=45, swing2LimitDegrees=30, twistLimitDegrees=2}
 {bone1Name=dyn_hoodie_back_3_1_jnt, bone2Name=spine_01, swing1LimitDegrees=45, swing2LimitDegrees=25, twistLimitDegrees=2}
 {bone1Name=dyn_sleeves_tie_jnt, bone2Name=spine_01, swing1LimitDegrees=10, swing2LimitDegrees=10, twistLimitDegrees=2}
 {bone1Name=dyn_grenade1_jnt, bone2Name=spine_01, swing1LimitDegrees=10, swing2LimitDegrees=40, twistLimitDegrees=2}
 {bone1Name=dyn_grenade2_jnt, bone2Name=spine_01, swing1LimitDegrees=10, swing2LimitDegrees=40, twistLimitDegrees=2}
 {
  bone1Name=spine_01
  bone2Name=pelvis
  swing1LimitDegrees=45
  swing1Motion=Free
  swing2LimitDegrees=45
  swing2Motion=Free
  twistLimitDegrees=45
  twistMotion=Free
 }
 {
  bone1Name=spine_02
  bone2Name=spine_01
  swing1LimitDegrees=45
  swing1Motion=Free
  swing2LimitDegrees=45
  swing2Motion=Free
  twistLimitDegrees=45
  twistMotion=Free
 }
 {
  bone1Name=thigh_l
  bone2Name=pelvis
  swing1LimitDegrees=45
  swing1Motion=Free
  swing2LimitDegrees=45
  swing2Motion=Free
  twistLimitDegrees=45
  twistMotion=Free
 }
 {
  bone1Name=thigh_r
  bone2Name=pelvis
  swing1LimitDegrees=45
  swing1Motion=Free
  swing2LimitDegrees=45
  swing2Motion=Free
  twistLimitDegrees=45
  twistMotion=Free
 }
 {
  bone1Name=dyn_hoodie_back_1_3_jnt
  bone2Name=dyn_hoodie_back_2_3_jnt
  swing1LimitDegrees=45
  swing1Motion=Free
  swing2LimitDegrees=45
  swing2Motion=Free
  twistLimitDegrees=45
  twistMotion=Free
 }
]
