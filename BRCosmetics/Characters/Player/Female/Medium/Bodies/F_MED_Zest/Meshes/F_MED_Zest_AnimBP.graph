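# Animation Blueprint for the F_MED_Zest female body cosmetic, adding secondary motion to tied sleeves, hoodie back and grenade bones.
# Separate branches from a cached pose use RigidBody nodes with standard, crouch and FE physics assets, or Anim Dynamics chains with spherical limits around thighs and pelvis.
# Bone transforms weighted by movement-direction variables, plus constraints and Look Ats driven by right-thigh lift, adjust the sleeves.
# FE and hoodie branches add fixed bone offsets.
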AnimBlueprint /BRCosmetics/Characters/Player/Female/Medium/Bodies/F_MED_Zest/Meshes/F_MED_Zest_AnimBP
target_skeleton=&/F_MED_Zest_Skeleton
graphs [
 AnimGraph {nodes [
   n1 = RigidBody(ComponentPose=Local To Component(LocalPose=Use cached pose 'Pose'())) {
    bClampLinearTranslationLimitToRefPose=true
    bOverrideWorldGravity=true
    overridePhysicsAsset=&/F_MED_Zest_CrouchPhysics
    simulationSpace=BaseBoneSpace
    alphaBoolBlend {blendOption=Linear}
    baseBoneRef {boneName=pelvis}
    componentAppliedLinearAccClamp=<3000 3000 3000>
    componentLinearAccScale=<0.3 0.2 0.2>
    componentLinearVelScale=<0.3 0.2 0.2>
    overrideWorldGravity {z=-980}
   }
   n2 = "Transform (Modify) Bone - Bone: dyn_sleeves_tie_jnt"!(ComponentPose=@n1, Alpha=1, Translation="-3.170,-0.906,-9.309", Rotation="-13.465,13.060,30.501") -> (Pose) {
    rotationMode=BMM_Replace
    rotationSpace=BCS_BoneSpace
    scaleMode=BMM_Additive
    scaleSpace=BCS_BoneSpace
    translationMode=BMM_Replace
    translationSpace=BCS_BoneSpace
    alphaBoolBlend {blendOption=Linear}
    boneToModify {boneName=dyn_sleeves_tie_jnt}
    rotation=<rot -13.465 13.0602 30.5012>
    translation=<-3.17 -0.906 -9.309>
   }
   n3 = "Constraint - dyn_sleeves_tie_jnt (PPPP)"!(
    ComponentPose=@n2
    Alpha=1
    ConstraintWeights_0=0.7467
    ConstraintWeights_1:float=0.0442
    ConstraintWeights_2:float=0.0158
    ConstraintWeights_3:float=0.1934
   ) -> (Pose) {
    alpha=0
    alphaBoolBlend {blendOption=Linear}
    boneToModify {boneName=dyn_sleeves_tie_jnt}
    constraintSetup [
     {targetBone {boneName=pelvis}}
     {targetBone {boneName=thigh_twist_01_r}}
     {targetBone {boneName=deform_thigh_fr_l}}
     {targetBone {boneName=deform_thigh_fr_r}}
    ]
    constraintWeights [1,1,1,1]
   }
   n4 = "Transform (Modify) Bone - Bone: dyn_sleeves_tie_jnt"(ComponentPose=@n3, Alpha=1, Translation="-2.076,-2.058,-9.792", Rotation="-21.486,16.809,16.468") {
    alpha=0
    rotationMode=BMM_Replace
    rotationSpace=BCS_BoneSpace
    scaleMode=BMM_Additive
    scaleSpace=BCS_BoneSpace
    translationMode=BMM_Replace
    translationSpace=BCS_BoneSpace
    alphaBoolBlend {blendOption=Linear}
    boneToModify {boneName=dyn_sleeves_tie_jnt}
    rotation=<rot -21.486 16.8092 16.4682>
    translation=<-2.076 -2.058 -9.792>
   }
   n5 = Get Forward_Direct() -> (Forward_Direct:double)
   n6 = "Transform (Modify) Bone - Bone: dyn_sleeves_tie_jnt"(ComponentPose=@n4, Alpha=@n5, Translation="0,0,0", Rotation="12.003,29.474,4.970") {
    rotationMode=BMM_Replace
    rotationSpace=BCS_BoneSpace
    scaleMode=BMM_Additive
    scaleSpace=BCS_BoneSpace
    translationMode=BMM_Additive
    translationSpace=BCS_BoneSpace
    alphaBoolBlend {blendOption=Linear}
    alphaScaleBiasClamp {bClampResult=true, bInterpResult=true, clampMax=0.8, interpSpeedDecreasing=25, interpSpeedIncreasing=30}
    boneToModify {boneName=dyn_sleeves_tie_jnt}
    rotation=<rot 12.003 29.474 4.97>
   }
   n7 = "Transform (Modify) Bone - Bone: dyn_L_sleeve_1_jnt"!(ComponentPose=@n6, Alpha=@n5, Rotation="48.989,41.652,2.747") -> (Pose) {
    rotationMode=BMM_Additive
    rotationSpace=BCS_BoneSpace
    scaleMode=BMM_Additive
    scaleSpace=BCS_BoneSpace
    translationMode=BMM_Additive
    translationSpace=BCS_BoneSpace
    alphaBoolBlend {blendOption=Linear}
    alphaScaleBiasClamp {bClampResult=true, bInterpResult=true, clampMax=0.8, interpSpeedDecreasing=25, interpSpeedIncreasing=30}
    boneToModify {boneName=dyn_L_sleeve_1_jnt}
    rotation=<rot 48.989 41.6521 2.747>
    translation=<-0.8123 -1.4757 4.0799>
   }
   n8 = "Transform (Modify) Bone - Bone: dyn_R_sleeve_1_jnt"!(ComponentPose=@n7, Alpha=@n5, Rotation="-23.789,23.050,29.355") -> (Pose) {
    rotationMode=BMM_Additive
    rotationSpace=BCS_BoneSpace
    scaleMode=BMM_Additive
    scaleSpace=BCS_BoneSpace
    translationMode=BMM_Additive
    translationSpace=BCS_BoneSpace
    alphaBoolBlend {blendOption=Linear}
    alphaScaleBiasClamp {bClampResult=true, bInterpResult=true, clampMax=0.8, interpSpeedDecreasing=25, interpSpeedIncreasing=30}
    boneToModify {boneName=dyn_R_sleeve_1_jnt}
    rotation=<rot -23.789 23.0499 29.355>
    translation=<-0.3556 -1.5959 -3.8805>
   }
   n9 = Get Backward_Direct() -> (Backward_Direct:double)
   n10 = "Transform (Modify) Bone - Bone: dyn_sleeves_tie_jnt"(ComponentPose=@n8, Alpha=@n9, Translation="-0.704,2.581,4.419", Rotation="33.624,39.696,-12.654") {
    rotationMode=BMM_Replace
    rotationSpace=BCS_BoneSpace
    scaleMode=BMM_Additive
    scaleSpace=BCS_BoneSpace
    translationMode=BMM_Additive
    translationSpace=BCS_BoneSpace
    alphaBoolBlend {blendOption=Linear}
    alphaScaleBiasClamp {bClampResult=true, bInterpResult=true, clampMax=0.8, interpSpeedDecreasing=60, interpSpeedIncreasing=60}
    boneToModify {boneName=dyn_sleeves_tie_jnt}
    rotation=<rot 33.624 39.6958 -12.6542>
    translation=<-0.704 2.581 4.419>
   }
   n11 = "Transform (Modify) Bone - Bone: dyn_L_sleeve_1_jnt"!(ComponentPose=@n10, Alpha=@n9, Translation:Vector="-2.228,-0.745,3.970", Rotation="3.980,7.245,-8.978") -> (Pose) {
    rotationMode=BMM_Replace
    rotationSpace=BCS_BoneSpace
    scaleMode=BMM_Additive
    scaleSpace=BCS_BoneSpace
    translationMode=BMM_Replace
    translationSpace=BCS_BoneSpace
    alphaBoolBlend {blendOption=Linear}
    alphaScaleBiasClamp {bClampResult=true, bInterpResult=true, clampMax=0.8, interpSpeedDecreasing=60, interpSpeedIncreasing=60}
    boneToModify {boneName=dyn_L_sleeve_1_jnt}
    rotation=<rot 3.98 7.245 -8.9781>
    translation=<-2.228 -0.745 3.97>
   }
   n12 = "Transform (Modify) Bone - Bone: dyn_R_sleeve_1_jnt"(ComponentPose=@n11, Alpha=@n9, Rotation="-63.673,129.785,-113.802") {
    rotationMode=BMM_Additive
    rotationSpace=BCS_BoneSpace
    scaleMode=BMM_Additive
    scaleSpace=BCS_BoneSpace
    translationMode=BMM_Additive
    translationSpace=BCS_BoneSpace
    alphaBoolBlend {blendOption=Linear}
    alphaScaleBiasClamp {bClampResult=true, bInterpResult=true, clampMax=0.8, interpSpeedDecreasing=25, interpSpeedIncreasing=30}
    boneToModify {boneName=dyn_R_sleeve_1_jnt}
    rotation=<rot -63.673 129.785 -113.8019>
    translation=<-0.3556 -1.5959 -3.8805>
   }
   n13 = Get Left_Direct() -> (Left_Direct:double)
   n14 = "Transform (Modify) Bone - Bone: dyn_sleeves_tie_jnt"(ComponentPose=@n12, Alpha=@n13, Translation="-0.952,0.193,3.964", Rotation="19.949,20.331,1.465") {
    rotationMode=BMM_Additive
    rotationSpace=BCS_BoneSpace
    scaleMode=BMM_Additive
    scaleSpace=BCS_BoneSpace
    translationMode=BMM_Additive
    translationSpace=BCS_BoneSpace
    alphaBoolBlend {blendOption=Linear}
    alphaScaleBiasClamp {bClampResult=true, bInterpResult=true, clampMax=0.8, interpSpeedDecreasing=25, interpSpeedIncreasing=30}
    boneToModify {boneName=dyn_sleeves_tie_jnt}
    rotation=<rot 19.949 20.331 1.465>
    translation=<-0.952 0.193 3.964>
   }
   n15 = "Transform (Modify) Bone - Bone: dyn_L_sleeve_1_jnt"!(ComponentPose=@n14, Alpha=@n13, Rotation="35.000,0.000,0.000") -> (Pose) {
    rotationMode=BMM_Additive
    rotationSpace=BCS_BoneSpace
    scaleMode=BMM_Additive
    scaleSpace=BCS_BoneSpace
    translationMode=BMM_Additive
    translationSpace=BCS_BoneSpace
    alphaBoolBlend {blendOption=Linear}
    alphaScaleBiasClamp {bClampResult=true, bInterpResult=true, clampMax=0.8, interpSpeedDecreasing=25, interpSpeedIncreasing=30}
    boneToModify {boneName=dyn_L_sleeve_1_jnt}
    rotation {pitch=35}
    translation=<0 0.0001 2.081>
   }
   n16 = "Transform (Modify) Bone - Bone: dyn_R_sleeve_1_jnt"!(ComponentPose=@n15, Alpha=@n13, Translation:Vector="-1.316,0.068,-1.566", Rotation="-4.179,8.806,-7.330") -> (Pose) {
    rotationMode=BMM_Replace
    rotationSpace=BCS_BoneSpace
    scaleMode=BMM_Additive
    scaleSpace=BCS_BoneSpace
    translationMode=BMM_Additive
    translationSpace=BCS_BoneSpace
    alphaBoolBlend {blendOption=Linear}
    alphaScaleBiasClamp {bClampResult=true, bInterpResult=true, clampMax=0.8, interpSpeedDecreasing=25, interpSpeedIncreasing=30}
    boneToModify {boneName=dyn_R_sleeve_1_jnt}
    rotation=<rot -4.179 8.8061 -7.3301>
    translation=<-1.316 0.068 -1.566>
   }
   n17 = Get Right_Direct() -> (Right_Direct:double)
   n18 = "Transform (Modify) Bone - Bone: dyn_sleeves_tie_jnt"(ComponentPose=@n16, Alpha=@n17, Translation="0,0,0", Rotation="11.192,29.495,14.764") {
    rotationMode=BMM_Additive
    rotationSpace=BCS_BoneSpace
    scaleMode=BMM_Additive
    scaleSpace=BCS_BoneSpace
    translationMode=BMM_Additive
    translationSpace=BCS_BoneSpace
    alphaBoolBlend {blendOption=Linear}
    alphaScaleBiasClamp {bClampResult=true, bInterpResult=true, clampMax=0.8, interpSpeedDecreasing=25, interpSpeedIncreasing=30}
    boneToModify {boneName=dyn_sleeves_tie_jnt}
    rotation=<rot 11.192 29.495 14.7641>
   }
   n19 = "Transform (Modify) Bone - Bone: dyn_L_sleeve_1_jnt"(ComponentPose=@n18, Alpha=@n17, Rotation="39.669,-4.046,2.151") {
    rotationMode=BMM_Additive
    rotationSpace=BCS_BoneSpace
    scaleMode=BMM_Additive
    scaleSpace=BCS_BoneSpace
    translationMode=BMM_Additive
    translationSpace=BCS_BoneSpace
    alphaBoolBlend {blendOption=Linear}
    alphaScaleBiasClamp {bClampResult=true, bInterpResult=true, clampMax=0.8, interpSpeedDecreasing=25, interpSpeedIncreasing=30}
    boneToModify {boneName=dyn_L_sleeve_1_jnt}
    rotation=<rot 39.669 -4.046 2.151>
   }
   n20 = "Transform (Modify) Bone - Bone: dyn_R_sleeve_1_jnt"!(ComponentPose=@n19, Alpha=@n17, Rotation="-39.385,7.358,-4.499") -> (Pose) {
    rotationMode=BMM_Additive
    rotationSpace=BCS_BoneSpace
    scaleMode=BMM_Additive
    scaleSpace=BCS_BoneSpace
    translationMode=BMM_Additive
    translationSpace=BCS_BoneSpace
    alphaBoolBlend {blendOption=Linear}
    alphaScaleBiasClamp {bClampResult=true, bInterpResult=true, clampMax=0.8, interpSpeedDecreasing=25, interpSpeedIncreasing=30}
    boneToModify {boneName=dyn_R_sleeve_1_jnt}
    rotation=<rot -39.385 7.358 -4.499>
   }
   n21 = Get thigh_rot_up_r() -> (thigh_rot_up_r:double)
   n22 = "Constraint - dyn_L_sleeve_1_jnt (P)"(ComponentPose=@n20, Alpha=@n21, ConstraintWeights_0=1) {
    alphaBoolBlend {blendOption=Linear}
    alphaScaleBiasClamp {bClampResult=true, clampMax=0.25}
    boneToModify {boneName=dyn_L_sleeve_1_jnt}
    constraintSetup [{targetBone {boneName=deform_thigh_fr_r}}]
    constraintWeights [1]
   }
   n23 = "Constraint - dyn_R_sleeve_1_jnt (P)"(ComponentPose=@n22, Alpha=@n21, ConstraintWeights_0=1) {
    alphaBoolBlend {blendOption=Linear}
    alphaScaleBiasClamp {bClampResult=true, clampMax=0.25}
    boneToModify {boneName=dyn_R_sleeve_1_jnt}
    constraintSetup [{targetBone {boneName=deform_thigh_fr_r}}]
    constraintWeights [1]
   }
   Component To Local(ComponentPose=@n23)
   n24 = "Transform (Modify) Bone - Bone: dyn_hoodie_back_1_1_jnt"(
    ComponentPose=Local To Component(LocalPose=Use cached pose 'Pose'())
    Alpha=0.8
    Translation="0,0,0"
    Rotation="-35.000,0.000,0.000"
    Scale="1,1,1"
   ) {
    rotationMode=BMM_Additive
    rotationSpace=BCS_BoneSpace
    scaleMode=BMM_Additive
    scaleSpace=BCS_BoneSpace
    translationMode=BMM_Additive
    translationSpace=BCS_BoneSpace
    alphaBoolBlend {blendOption=Linear}
    boneToModify {boneName=dyn_hoodie_back_1_1_jnt}
    rotation {pitch=-35}
   }
   n25 = "Transform (Modify) Bone - Bone: dyn_hoodie_back_2_1_jnt"(ComponentPose=@n24, Alpha=0.8, Translation="0,0,0", Rotation="0,-30,0", Scale="1,1,1") {
    rotationMode=BMM_Additive
    rotationSpace=BCS_BoneSpace
    scaleMode=BMM_Additive
    scaleSpace=BCS_BoneSpace
    translationMode=BMM_Additive
    translationSpace=BCS_BoneSpace
    alphaBoolBlend {blendOption=Linear}
    boneToModify {boneName=dyn_hoodie_back_2_1_jnt}
    rotation {yaw=-30}
   }
   n26 = "Transform (Modify) Bone - Bone: dyn_hoodie_back_3_1_jnt"(ComponentPose=@n25, Alpha=0.8, Translation="0,0,0", Rotation="0,-20,0", Scale="1,1,1") {
    rotationMode=BMM_Additive
    rotationSpace=BCS_BoneSpace
    scaleMode=BMM_Additive
    scaleSpace=BCS_BoneSpace
    translationMode=BMM_Additive
    translationSpace=BCS_BoneSpace
    alphaBoolBlend {blendOption=Linear}
    boneToModify {boneName=dyn_hoodie_back_3_1_jnt}
    rotation {yaw=-20}
   }
   n27 = Anim Dynamics - Bone: dyn_hoodie_back_1_1_jnt(ComponentPose=@n26, Alpha=1, ExternalForce="0,0,0") {
    angularDampingOverride=0.9999
    angularSpringConstant=100
    bAngularSpring=true
    bOverrideAngularDamping=true
    bUseSphericalLimits=true
    simulationSpace=RootRelative
    alphaBoolBlend {blendOption=Linear}
    boundBone {boneName=dyn_hoodie_back_1_1_jnt}
    componentAppliedLinearAccClamp=<3500 3500 3500>
    componentLinearAccScale=<1.2 1.2 1.2>
    componentLinearVelScale=<1.2 1.2 1.2>
    physicsBodyDefinitions [{
      sphereCollisionRadius=0
      boundBone {boneName=dyn_hoodie_back_1_1_jnt}
      boxExtents=<20 10 10>
      constraintSetup {angularLimitsMax=<1 5 15>, angularLimitsMin=<-1 -30 -15>, angularTarget=<0 0 0>}
      localJointOffset=<2 0 0>
     }]
    relativeSpaceBone {boneName=pelvis}
    sphericalLimits [{limitRadius=5, drivingBone {boneName=thigh_r}, sphereLocalOffset=<0 0 0>}]
   }
   n28 = Anim Dynamics - Bone: dyn_hoodie_back_1_2_jnt(ComponentPose=@n27, Alpha=1, ExternalForce="0,0,0") {
    angularDampingOverride=0.9999
    angularSpringConstant=100
    bAngularSpring=true
    bOverrideAngularDamping=true
    bUseSphericalLimits=true
    simulationSpace=RootRelative
    alphaBoolBlend {blendOption=Linear}
    boundBone {boneName=dyn_hoodie_back_1_2_jnt}
    componentAppliedLinearAccClamp=<3500 3500 3500>
    componentLinearAccScale=<1.2 1.2 1.2>
    componentLinearVelScale=<1.2 1.2 1.2>
    physicsBodyDefinitions [{
      sphereCollisionRadius=0
      boundBone {boneName=dyn_hoodie_back_1_2_jnt}
      boxExtents=<20 10 10>
      constraintSetup {angularLimitsMax=<1 5 15>, angularLimitsMin=<-1 -30 -15>, angularTarget=<0 0 0>}
      localJointOffset=<2 0 0>
     }]
    relativeSpaceBone {boneName=pelvis}
    sphericalLimits [{limitRadius=8, drivingBone {boneName=deform_thigh_fr_r}, sphereLocalOffset=<6 0 0>}]
   }
   n29 = Anim Dynamics - Bone: dyn_hoodie_back_1_3_jnt(ComponentPose=@n28, Alpha=1, ExternalForce="0,0,0") {
    angularDampingOverride=0.9999
    angularSpringConstant=100
    bAngularSpring=true
    bOverrideAngularDamping=true
    bUseSphericalLimits=true
    simulationSpace=RootRelative
    alphaBoolBlend {blendOption=Linear}
    boundBone {boneName=dyn_hoodie_back_1_3_jnt}
    componentAppliedLinearAccClamp=<3500 3500 3500>
    componentLinearAccScale=<1.2 1.2 1.2>
    componentLinearVelScale=<1.2 1.2 1.2>
    physicsBodyDefinitions [{
      sphereCollisionRadius=0
      boundBone {boneName=dyn_hoodie_back_1_3_jnt}
      boxExtents=<20 10 10>
      constraintSetup {angularLimitsMax=<1 10 15>, angularLimitsMin=<-1 -30 -15>, angularTarget=<0 0 0>}
      localJointOffset=<2 0 0>
     }]
    relativeSpaceBone {boneName=pelvis}
    sphericalLimits [{limitRadius=8, drivingBone {boneName=deform_thigh_fr_r}, sphereLocalOffset=<10 0 0>}]
   }
   n30 = Anim Dynamics - Bone: dyn_hoodie_back_2_1_jnt(ComponentPose=@n29, Alpha=1, ExternalForce="0,0,0") {
    angularDampingOverride=0.9999
    angularSpringConstant=100
    bAngularSpring=true
    bOverrideAngularDamping=true
    bUseSphericalLimits=true
    simulationSpace=RootRelative
    alphaBoolBlend {blendOption=Linear}
    boundBone {boneName=dyn_hoodie_back_2_1_jnt}
    componentAppliedLinearAccClamp=<3500 3500 3500>
    componentLinearAccScale=<1.2 1.2 1.2>
    componentLinearVelScale=<1.2 1.2 1.2>
    physicsBodyDefinitions [{
      sphereCollisionRadius=0
      boundBone {boneName=dyn_hoodie_back_2_1_jnt}
      boxExtents=<20 10 10>
      constraintSetup {angularLimitsMax=<1 5 15>, angularLimitsMin=<-1 -30 -15>, angularTarget=<0 0 0>}
      localJointOffset=<2 0 0>
     }]
    relativeSpaceBone {boneName=pelvis}
    sphericalLimits [{limitRadius=7, drivingBone {boneName=thigh_r}, sphereLocalOffset=<0 2 0>}]
   }
   n31 = Anim Dynamics - Bone: dyn_hoodie_back_2_2_jnt(ComponentPose=@n30, Alpha=1, ExternalForce="0,0,0") {
    angularDampingOverride=0.9999
    angularSpringConstant=100
    bAngularSpring=true
    bOverrideAngularDamping=true
    bUseSphericalLimits=true
    simulationSpace=RootRelative
    alphaBoolBlend {blendOption=Linear}
    boundBone {boneName=dyn_hoodie_back_2_2_jnt}
    componentAppliedLinearAccClamp=<3500 3500 3500>
    componentLinearAccScale=<1.2 1.2 1.2>
    componentLinearVelScale=<1.2 1.2 1.2>
    physicsBodyDefinitions [{
      sphereCollisionRadius=0
      boundBone {boneName=dyn_hoodie_back_2_2_jnt}
      boxExtents=<20 10 10>
      constraintSetup {angularLimitsMax=<1 5 15>, angularLimitsMin=<-1 -30 -15>, angularTarget=<0 0 0>}
      localJointOffset=<2 0 0>
     }]
    relativeSpaceBone {boneName=pelvis}
    sphericalLimits [{limitRadius=7, drivingBone {boneName=deform_thigh_bk_r}, sphereLocalOffset=<0 3 0>}]
   }
   n32 = Anim Dynamics - Bone: dyn_hoodie_back_2_3_jnt(ComponentPose=@n31, Alpha=1, ExternalForce="0,0,0") {
    angularDampingOverride=0.9999
    angularSpringConstant=100
    bAngularSpring=true
    bOverrideAngularDamping=true
    bUseSphericalLimits=true
    simulationSpace=RootRelative
    alphaBoolBlend {blendOption=Linear}
    boundBone {boneName=dyn_hoodie_back_2_3_jnt}
    componentAppliedLinearAccClamp=<3500 3500 3500>
    componentLinearAccScale=<1.2 1.2 1.2>
    componentLinearVelScale=<1.2 1.2 1.2>
    physicsBodyDefinitions [{
      sphereCollisionRadius=0
      boundBone {boneName=dyn_hoodie_back_2_3_jnt}
      boxExtents=<20 10 10>
      constraintSetup {angularLimitsMax=<1 5 15>, angularLimitsMin=<-1 -30 -15>, angularTarget=<0 0 0>}
      localJointOffset=<2 0 0>
     }]
    relativeSpaceBone {boneName=pelvis}
    sphericalLimits [{limitRadius=7, drivingBone {boneName=deform_thigh_bk_r}, sphereLocalOffset=<5 3 0>}]
   }
   n33 = Anim Dynamics - Bone: dyn_hoodie_back_3_1_jnt(ComponentPose=@n32, Alpha=1, ExternalForce="0,0,0") {
    angularDampingOverride=0.9999
    angularSpringConstant=100
    bAngularSpring=true
    bOverrideAngularDamping=true
    bUseSphericalLimits=true
    simulationSpace=RootRelative
    alphaBoolBlend {blendOption=Linear}
    boundBone {boneName=dyn_hoodie_back_3_1_jnt}
    componentAppliedLinearAccClamp=<3500 3500 3500>
    componentLinearAccScale=<1.2 1.2 1.2>
    componentLinearVelScale=<1.2 1.2 1.2>
    physicsBodyDefinitions [{
      sphereCollisionRadius=0
      boundBone {boneName=dyn_hoodie_back_3_1_jnt}
      boxExtents=<20 10 10>
      constraintSetup {angularLimitsMax=<1 10 15>, angularLimitsMin=<-1 -30 -15>, angularTarget=<0 0 0>}
      localJointOffset=<2 0 0>
     }]
    relativeSpaceBone {boneName=pelvis}
    sphericalLimits [{limitRadius=5, drivingBone {boneName=thigh_r}, sphereLocalOffset=<0 3 0>}]
   }
   n34 = Anim Dynamics - Bone: dyn_hoodie_back_3_2_jnt(ComponentPose=@n33, Alpha=1, ExternalForce="0,0,0") {
    angularDampingOverride=0.9999
    angularSpringConstant=100
    bAngularSpring=true
    bOverrideAngularDamping=true
    bUseSphericalLimits=true
    simulationSpace=RootRelative
    alphaBoolBlend {blendOption=Linear}
    boundBone {boneName=dyn_hoodie_back_3_2_jnt}
    componentAppliedLinearAccClamp=<3500 3500 3500>
    componentLinearAccScale=<1.2 1.2 1.2>
    componentLinearVelScale=<1.2 1.2 1.2>
    physicsBodyDefinitions [{
      sphereCollisionRadius=0
      boundBone {boneName=dyn_hoodie_back_3_2_jnt}
      boxExtents=<20 10 10>
      constraintSetup {angularLimitsMax=<1 10 15>, angularLimitsMin=<-1 -30 -15>, angularTarget=<0 0 0>}
      localJointOffset=<2 0 0>
     }]
    relativeSpaceBone {boneName=pelvis}
    sphericalLimits [{limitRadius=5, drivingBone {boneName=thigh_r}, sphereLocalOffset=<5 4 0>}]
   }
   n35 = Anim Dynamics - Bone: dyn_grenade1_jnt(ComponentPose=@n34, Alpha=0.5, ExternalForce="0,0,0") {
    angularSpringConstant=1500
    bGravityOverrideInSimSpace=true
    bUsePlanarLimit=false
    bUseSphericalLimits=true
    gravityScale=0.1
    linearSpringConstant=1000
    simulationSpace=BoneRelative
    alphaBoolBlend {blendOption=Linear}
    boundBone {boneName=dyn_grenade1_jnt}
    componentLinearAccScale=<0.15 0.15 0.35>
    componentLinearVelScale=<0.15 15 0.35>
    gravityOverride {z=-980}
    physicsBodyDefinitions [{
      sphereCollisionRadius=0
      boundBone {boneName=dyn_grenade1_jnt}
      boxExtents=<8 20 20>
      constraintSetup {angularLimitsMax=<2 5 30>, angularLimitsMin=<-5 -10 -35>, angularTarget=<0 0 0>}
      localJointOffset=<10 0 0>
     }]
    sphericalLimits [{limitRadius=7, drivingBone {boneName=dyn_grenade2_jnt}, sphereLocalOffset=<8 0 0>}]
   }
   n36 = Anim Dynamics - Bone: dyn_grenade2_jnt(ComponentPose=@n35, Alpha=0.5, ExternalForce="0,0,0") {
    angularSpringConstant=1500
    bGravityOverrideInSimSpace=true
    bUseSphericalLimits=true
    gravityScale=0.1
    linearSpringConstant=1000
    simulationSpace=BoneRelative
    alphaBoolBlend {blendOption=Linear}
    boundBone {boneName=dyn_grenade2_jnt}
    componentLinearAccScale=<0.15 0.15 0.35>
    componentLinearVelScale=<0.15 15 0.35>
    gravityOverride {z=-980}
    physicsBodyDefinitions [{
      sphereCollisionRadius=0
      boundBone {boneName=dyn_grenade2_jnt}
      boxExtents=<8 20 20>
      constraintSetup {angularLimitsMax=<2 5 30>, angularLimitsMin=<-5 -10 -35>, angularTarget=<0 0 0>}
      localJointOffset=<10 0 0>
     }]
    sphericalLimits [{limitRadius=7, drivingBone {boneName=dyn_grenade1_jnt}, sphereLocalOffset=<8 0 0>}]
   }
   n37 = Anim Dynamics - Bone: dyn_L_sleeve_1_jnt(ComponentPose=@n36, Alpha=1, ExternalForce="0,0,0") {
    angularDampingOverride=0.9999
    angularSpringConstant=100
    bAngularSpring=true
    bOverrideAngularDamping=true
    bUseSphericalLimits=true
    simulationSpace=RootRelative
    alphaBoolBlend {blendOption=Linear}
    boundBone {boneName=dyn_L_sleeve_1_jnt}
    componentAppliedLinearAccClamp=<3500 3500 3500>
    componentLinearAccScale=<1.2 1.2 1.2>
    componentLinearVelScale=<1.2 1.2 1.2>
    physicsBodyDefinitions [{
      sphereCollisionRadius=0
      boundBone {boneName=dyn_L_sleeve_1_jnt}
      boxExtents=<20 10 10>
      constraintSetup {angularLimitsMax=<1 15 15>, angularLimitsMin=<-1 -60 -15>, angularTarget=<0 0 0>}
      localJointOffset=<5 0 0>
     }]
    relativeSpaceBone {boneName=dyn_sleeves_tie_jnt}
    sphericalLimits [{limitRadius=7, drivingBone {boneName=deform_thigh_fr_l}, sphereLocalOffset=<0 0 0>}]
   }
   n38 = Anim Dynamics - Bone: dyn_L_sleeve_2_jnt(ComponentPose=@n37, Alpha=1, ExternalForce="0,0,0") {
    angularDampingOverride=0.9999
    angularSpringConstant=100
    bAngularSpring=true
    bOverrideAngularDamping=true
    bUseSphericalLimits=true
    simulationSpace=RootRelative
    alphaBoolBlend {blendOption=Linear}
    boundBone {boneName=dyn_L_sleeve_2_jnt}
    componentAppliedLinearAccClamp=<3500 3500 3500>
    componentLinearAccScale=<1.2 1.2 1.2>
    componentLinearVelScale=<1.2 1.2 1.2>
    physicsBodyDefinitions [{
      sphereCollisionRadius=0
      boundBone {boneName=dyn_L_sleeve_2_jnt}
      boxExtents=<20 10 10>
      constraintSetup {angularLimitsMax=<1 15 15>, angularLimitsMin=<-1 -60 -15>, angularTarget=<0 0 0>}
      localJointOffset=<5 0 0>
     }]
    relativeSpaceBone {boneName=dyn_sleeves_tie_jnt}
    sphericalLimits [{limitRadius=7, drivingBone {boneName=deform_thigh_fr_l}, sphereLocalOffset=<-8 0 0>}]
   }
   n39 = Anim Dynamics - Bone: dyn_L_sleeve_3_jnt(ComponentPose=@n38, Alpha=1, ExternalForce="0,0,0") {
    angularDampingOverride=0.9999
    angularSpringConstant=100
    bAngularSpring=true
    bOverrideAngularDamping=true
    bUseSphericalLimits=true
    simulationSpace=RootRelative
    alphaBoolBlend {blendOption=Linear}
    boundBone {boneName=dyn_L_sleeve_3_jnt}
    componentAppliedLinearAccClamp=<3500 3500 3500>
    componentLinearAccScale=<1.2 1.2 1.2>
    componentLinearVelScale=<1.2 1.2 1.2>
    physicsBodyDefinitions [{
      sphereCollisionRadius=0
      boundBone {boneName=dyn_L_sleeve_3_jnt}
      boxExtents=<20 10 10>
      constraintSetup {angularLimitsMax=<1 15 15>, angularLimitsMin=<-1 -60 -15>, angularTarget=<0 0 0>}
      localJointOffset=<5 0 0>
     }]
    relativeSpaceBone {boneName=dyn_sleeves_tie_jnt}
    sphericalLimits [{limitRadius=8, drivingBone {boneName=deform_thigh_fr_l}, sphereLocalOffset=<-10 0 0>}]
   }
   n40 = Anim Dynamics - Bone: dyn_R_sleeve_1_jnt(ComponentPose=@n39, Alpha=1, ExternalForce="0,0,0") {
    angularDampingOverride=0.9999
    angularSpringConstant=100
    bAngularSpring=true
    bOverrideAngularDamping=true
    bUseSphericalLimits=true
    simulationSpace=RootRelative
    alphaBoolBlend {blendOption=Linear}
    boundBone {boneName=dyn_R_sleeve_1_jnt}
    componentAppliedLinearAccClamp=<3500 3500 3500>
    componentLinearAccScale=<1.2 1.2 1.2>
    componentLinearVelScale=<1.2 1.2 1.2>
    physicsBodyDefinitions [{
      sphereCollisionRadius=0
      boundBone {boneName=dyn_R_sleeve_1_jnt}
      boxExtents=<20 10 10>
      constraintSetup {angularLimitsMax=<1 15 15>, angularLimitsMin=<-1 -60 -15>, angularTarget=<0 0 0>}
      localJointOffset=<5 0 0>
     }]
    relativeSpaceBone {boneName=dyn_sleeves_tie_jnt}
    sphericalLimits [{limitRadius=7, drivingBone {boneName=pelvis}, sphereLocalOffset=<0 0 0>}]
   }
   n41 = Anim Dynamics - Bone: dyn_R_sleeve_2_jnt(ComponentPose=@n40, Alpha=1, ExternalForce="0,0,0") {
    angularDampingOverride=0.9999
    angularSpringConstant=100
    bAngularSpring=true
    bOverrideAngularDamping=true
    bUseSphericalLimits=true
    simulationSpace=RootRelative
    alphaBoolBlend {blendOption=Linear}
    boundBone {boneName=dyn_R_sleeve_2_jnt}
    componentAppliedLinearAccClamp=<3500 3500 3500>
    componentLinearAccScale=<1.2 1.2 1.2>
    componentLinearVelScale=<1.2 1.2 1.2>
    physicsBodyDefinitions [{
      sphereCollisionRadius=0
      boundBone {boneName=dyn_R_sleeve_2_jnt}
      boxExtents=<20 10 10>
      constraintSetup {angularLimitsMax=<1 15 15>, angularLimitsMin=<-1 -60 -15>, angularTarget=<0 0 0>}
      localJointOffset=<5 0 0>
     }]
    relativeSpaceBone {boneName=dyn_sleeves_tie_jnt}
    sphericalLimits [{limitRadius=7, drivingBone {boneName=pelvis}, sphereLocalOffset=<-8 0 0>}]
   }
   n42 = "Transform (Modify) Bone - Bone: dyn_sleeves_tie_jnt"(ComponentPose=@n41, Alpha=1, Translation="-3.436,0.086,-5.775", Rotation="-16.667,31.231,5.312") {
    alpha=0
    rotationMode=BMM_Replace
    rotationSpace=BCS_BoneSpace
    scaleMode=BMM_Additive
    scaleSpace=BCS_BoneSpace
    translationMode=BMM_Replace
    translationSpace=BCS_BoneSpace
    alphaBoolBlend {blendOption=Linear}
    boneToModify {boneName=dyn_sleeves_tie_jnt}
    rotation=<rot -16.667 31.2311 5.312>
    translation=<-3.436 0.086 -5.775>
   }
   n43 = Look At - Bone: dyn_sleeves_tie_jnt(ComponentPose=@n42, Alpha=Get thigh_rot_up_r()) {
    alphaBoolBlend {blendOption=Linear}
    boneToModify {boneName=dyn_sleeves_tie_jnt}
    lookAtLocation {x=120}
    lookAtTarget {boneReference {boneName=thigh_r}}
    lookAt_Axis {axis=<1 0>}
    lookUp_Axis {axis {x=0}}
   }
   n44 = Component To Local(ComponentPose=@n43)
   n45 = RigidBody(ComponentPose=Local To Component(LocalPose=Use cached pose 'Pose'())) {
    bClampLinearTranslationLimitToRefPose=true
    bOverrideWorldGravity=true
    overridePhysicsAsset=&/F_MED_Zest_Physics
    simulationSpace=BaseBoneSpace
    alphaBoolBlend {blendOption=Linear}
    baseBoneRef {boneName=pelvis}
    componentAppliedLinearAccClamp=<3000 3000 3000>
    componentLinearAccScale=<0.3 0.2 0.2>
    componentLinearVelScale=<0.3 0.2 0.2>
    overrideWorldGravity {z=-980}
   }
   n46 = Component To Local(ComponentPose=@n45)
   n47 = Get bRigidBodyEnabled()
   n48 = RigidBody(ComponentPose=Local To Component(LocalPose=Use cached pose 'Pose'())) {
    bClampLinearTranslationLimitToRefPose=true
    bOverrideWorldGravity=true
    overridePhysicsAsset=&/F_MED_Zest_FEPhysics
    simulationSpace=BaseBoneSpace
    alphaBoolBlend {blendOption=Linear}
    baseBoneRef {boneName=pelvis}
    componentAppliedLinearAccClamp=<3000 3000 3000>
    componentLinearAccScale=<0.3 0.2 0.2>
    componentLinearVelScale=<0.3 0.2 0.2>
    overrideWorldGravity {z=-980}
   }
   n49 = Look At - Bone: dyn_hoodie_back_2_1_jnt(ComponentPose=@n48, Alpha=1) {
    alpha=0
    alphaBoolBlend {blendOption=Linear}
    alphaScaleBiasClamp {bInterpResult=true, interpSpeedDecreasing=5, interpSpeedIncreasing=7}
    boneToModify {boneName=dyn_hoodie_back_2_1_jnt}
    lookAtLocation=<-45.0026 -30.5685 19.4021>
    lookAtTarget {boneReference {boneName=pelvis}}
    lookAt_Axis {axis=<1 0>}
    lookUp_Axis {axis {x=0}}
   }
   n50 = Look At - Bone: dyn_hoodie_back_3_1_jnt(ComponentPose=@n49, Alpha=1) {
    alpha=0
    alphaBoolBlend {blendOption=Linear}
    alphaScaleBiasClamp {bInterpResult=true, interpSpeedDecreasing=5, interpSpeedIncreasing=7}
    boneToModify {boneName=dyn_hoodie_back_3_1_jnt}
    lookAtLocation=<-45.001 -19.735 -14.0961>
    lookAtTarget {boneReference {boneName=pelvis}}
    lookAt_Axis {axis=<1 0>}
    lookUp_Axis {axis {x=0}}
   }
   n51 = "Transform (Modify) Bone - Bone: dyn_hoodie_back_2_1_jnt"(ComponentPose=@n50, Alpha=1, Translation="0,-3.973,-0.000", Rotation="0.000,5.000,0.000", Scale="0.93,1,1") {
    rotationMode=BMM_Replace
    rotationSpace=BCS_BoneSpace
    scaleMode=BMM_Additive
    scaleSpace=BCS_BoneSpace
    translationMode=BMM_Replace
    translationSpace=BCS_BoneSpace
    alphaBoolBlend {blendOption=Linear}
    boneToModify {boneName=dyn_hoodie_back_2_1_jnt}
    rotation {yaw=5}
    scale {x=0.93}
    translation {y=-3.973}
   }
   n52 = "Transform (Modify) Bone - Bone: dyn_hoodie_back_2_2_jnt"(ComponentPose=@n51, Alpha=1, Translation="-0.464,3.319,-0.000", Rotation="0.000,10,0.000", Scale="0.92,1,1") {
    rotationMode=BMM_Replace
    rotationSpace=BCS_BoneSpace
    scaleMode=BMM_Additive
    scaleSpace=BCS_BoneSpace
    translationMode=BMM_Replace
    translationSpace=BCS_BoneSpace
    alphaBoolBlend {blendOption=Linear}
    boneToModify {boneName=dyn_hoodie_back_2_2_jnt}
    rotation {yaw=10}
    scale {x=0.92}
    translation=<-0.464 3.319>
   }
   n53 = "Transform (Modify) Bone - Bone: dyn_hoodie_back_3_1_jnt"(ComponentPose=@n52, Alpha=1, Translation="0.000,-1.775,0.000", Rotation="0,0,0", Scale="1.1,1.000,1.000") {
    rotationMode=BMM_Replace
    rotationSpace=BCS_BoneSpace
    scaleMode=BMM_Additive
    scaleSpace=BCS_BoneSpace
    translationMode=BMM_Replace
    translationSpace=BCS_BoneSpace
    alphaBoolBlend {blendOption=Linear}
    boneToModify {boneName=dyn_hoodie_back_3_1_jnt}
    scale {x=1.1}
    translation {y=-1.775}
   }
   n54 = "Transform (Modify) Bone - Bone: dyn_hoodie_back_3_2_jnt"(ComponentPose=@n53, Alpha=1, Translation="0,0,0", Rotation="0.053,-18.228,-0.272", Scale="0.5,1.000,1.000") {
    rotationMode=BMM_Replace
    rotationSpace=BCS_BoneSpace
    scaleMode=BMM_Additive
    scaleSpace=BCS_BoneSpace
    translationMode=BMM_Replace
    translationSpace=BCS_BoneSpace
    alphaBoolBlend {blendOption=Linear}
    boneToModify {boneName=dyn_hoodie_back_3_2_jnt}
    rotation=<rot 0.053 -18.2281 -0.272>
    scale {x=0.5}
   }
   n55 = "Transform (Modify) Bone - Bone: dyn_hoodie_back_1_1_jnt"(ComponentPose=@n54, Alpha=1, Translation="0.084,-0.843,-0.146", Rotation="4.114,0.063,-10.393", Scale="1,1,1") {
    rotationMode=BMM_Replace
    rotationSpace=BCS_BoneSpace
    scaleMode=BMM_Additive
    scaleSpace=BCS_BoneSpace
    translationMode=BMM_Replace
    translationSpace=BCS_BoneSpace
    alphaBoolBlend {blendOption=Linear}
    boneToModify {boneName=dyn_hoodie_back_1_1_jnt}
    rotation=<rot 4.114 0.063 -10.3932>
    translation=<0.084 -0.843 -0.146>
   }
   n56 = "Constraint - dyn_hoodie_back_1_2_jnt (P)"(ComponentPose=@n55, Alpha=1, ConstraintWeights_0=1) {
    alpha=0
    alphaBoolBlend {blendOption=Linear}
    boneToModify {boneName=dyn_hoodie_back_1_2_jnt}
    constraintSetup [{targetBone {boneName=thigh_r}}]
    constraintWeights [1]
   }
   n57 = "Transform (Modify) Bone - Bone: dyn_hoodie_back_1_2_jnt"(ComponentPose=@n56, Alpha=1, Translation="0.15,0.000,0.5", Rotation="-5.000,11,0.000", Scale="1,1,1") {
    alpha=0
    rotationMode=BMM_Replace
    rotationSpace=BCS_BoneSpace
    scaleMode=BMM_Additive
    scaleSpace=BCS_BoneSpace
    translationMode=BMM_Replace
    translationSpace=BCS_BoneSpace
    alphaBoolBlend {blendOption=Linear}
    boneToModify {boneName=dyn_hoodie_back_1_2_jnt}
    rotation {pitch=-5, yaw=11}
    translation {x=0.15, z=0.5}
   }
   n58 = "Transform (Modify) Bone - Bone: dyn_R_sleeve_1_jnt"!(ComponentPose=@n57, Alpha=1, Translation:Vector="0,0,0", Rotation="0.000,-20.000,0.001", Scale:Vector="1,1,1") -> (Pose) {
    rotationMode=BMM_Replace
    rotationSpace=BCS_BoneSpace
    scaleMode=BMM_Additive
    scaleSpace=BCS_BoneSpace
    translationMode=BMM_Replace
    translationSpace=BCS_BoneSpace
    alphaBoolBlend {blendOption=Linear}
    boneToModify {boneName=dyn_R_sleeve_1_jnt}
    rotation {roll=0.001, yaw=-20}
   }
   n59 = Look At - Bone: dyn_hoodie_back_1_2_jnt(ComponentPose=@n58, Alpha=1) {
    alpha=0
    alphaBoolBlend {blendOption=Linear}
    alphaScaleBiasClamp {bInterpResult=true, interpSpeedDecreasing=5, interpSpeedIncreasing=7}
    boneToModify {boneName=dyn_hoodie_back_1_2_jnt}
    lookAtLocation {x=-45, z=30}
    lookAtTarget {boneReference {boneName=pelvis}}
    lookAt_Axis {axis=<1 0>}
    lookUp_Axis {axis {x=0}}
   }
   n60 = "Transform (Modify) Bone - Bone: dyn_hoodie_back_1_1_jnt"(ComponentPose=@n59, Alpha=1, Translation="0,0.000,-1", Rotation="0.8,-4,10", Scale="1,1,1") {
    alpha=0
    rotationMode=BMM_Replace
    rotationSpace=BCS_BoneSpace
    scaleMode=BMM_Additive
    scaleSpace=BCS_BoneSpace
    translationMode=BMM_Replace
    translationSpace=BCS_BoneSpace
    alphaBoolBlend {blendOption=Linear}
    boneToModify {boneName=dyn_hoodie_back_1_1_jnt}
    rotation=<rot 0.8 -4 10>
    translation {z=-1}
   }
   n61 = Look At - Bone: dyn_hoodie_back_2_2_jnt(ComponentPose=@n60, Alpha=1) {
    alpha=0
    alphaBoolBlend {blendOption=Linear}
    alphaScaleBiasClamp {bInterpResult=true, interpSpeedDecreasing=5, interpSpeedIncreasing=7}
    boneToModify {boneName=dyn_hoodie_back_2_2_jnt}
    lookAtLocation=<-45 -30 20>
    lookAtTarget {boneReference {boneName=pelvis}}
    lookAt_Axis {axis=<1 0>}
    lookUp_Axis {axis {x=0}}
   }
   n62 = Look At - Bone: dyn_hoodie_back_3_2_jnt(ComponentPose=@n61, Alpha=1) {
    alpha=0
    alphaBoolBlend {blendOption=Linear}
    alphaScaleBiasClamp {bInterpResult=true, interpSpeedDecreasing=5, interpSpeedIncreasing=7}
    boneToModify {boneName=dyn_hoodie_back_3_2_jnt}
    lookAtLocation=<-45 -35 -7>
    lookAtTarget {boneReference {boneName=pelvis}}
    lookAt_Axis {axis=<1 0>}
    lookUp_Axis {axis {x=0}}
   }
   n63 = RigidBody(ComponentPose=Local To Component(LocalPose=Use cached pose 'Pose'())) {
    bClampLinearTranslationLimitToRefPose=true
    bOverrideWorldGravity=true
    overridePhysicsAsset=&F_MED_Zest_CrouchPhysics
    simulationSpace=BaseBoneSpace
    alphaBoolBlend {blendOption=Linear}
    baseBoneRef {boneName=pelvis}
    componentAppliedLinearAccClamp=<3000 3000 3000>
    componentLinearAccScale=<0.3 0.2 0.2>
    componentLinearVelScale=<0.3 0.2 0.2>
    overrideWorldGravity {z=-980}
   }
   n64 = "Transform (Modify) Bone - Bone: dyn_sleeves_tie_jnt"(ComponentPose=@n63, Alpha=Get Moving() -> (Moving:double), Translation="-3.436,0.086,-5.775", Rotation="-16.667,31.231,5.312") {
    alpha=0
    rotationMode=BMM_Replace
    rotationSpace=BCS_BoneSpace
    scaleMode=BMM_Additive
    scaleSpace=BCS_BoneSpace
    translationMode=BMM_Replace
    translationSpace=BCS_BoneSpace
    alphaBoolBlend {blendOption=Linear}
    alphaScaleBiasClamp {bInterpResult=true, interpSpeedDecreasing=7}
    boneToModify {boneName=dyn_sleeves_tie_jnt}
    rotation=<rot -16.667 31.2311 5.312>
    translation=<-3.436 0.086 -5.775>
   }
   n65 = Anim Dynamics - Bone: dyn_L_sleeve_1_jnt(ComponentPose=@n64, Alpha=1, ExternalForce="0,0,0") {
    angularDampingOverride=0.9999
    angularSpringConstant=100
    bAngularSpring=true
    bOverrideAngularDamping=true
    bUseSphericalLimits=true
    simulationSpace=RootRelative
    alphaBoolBlend {blendOption=Linear}
    boundBone {boneName=dyn_L_sleeve_1_jnt}
    componentAppliedLinearAccClamp=<3500 3500 3500>
    componentLinearAccScale=<1 1 1>
    componentLinearVelScale=<1 1 1>
    gravityOverride {z=-980}
    physicsBodyDefinitions [{
      sphereCollisionRadius=0
      boundBone {boneName=dyn_L_sleeve_1_jnt}
      boxExtents=<10 10 20>
      constraintSetup {angularLimitsMax=<1 30 15>, angularLimitsMin=<-1 -45 -15>, angularTarget=<0 0 0>}
      localJointOffset=<2 0 0>
     }]
    relativeSpaceBone {boneName=dyn_sleeves_tie_jnt}
    sphericalLimits [{limitRadius=7, drivingBone {boneName=deform_thigh_fr_l}, sphereLocalOffset=<0 0 0>}]
   }
   n66 = Anim Dynamics - Bone: dyn_L_sleeve_2_jnt(ComponentPose=@n65, Alpha=1, ExternalForce="0,0,0") {
    angularDampingOverride=0.9999
    angularSpringConstant=100
    bAngularSpring=true
    bOverrideAngularDamping=true
    bUseSphericalLimits=true
    simulationSpace=RootRelative
    alphaBoolBlend {blendOption=Linear}
    boundBone {boneName=dyn_L_sleeve_2_jnt}
    componentAppliedLinearAccClamp=<3500 3500 3500>
    componentLinearAccScale=<1 1 1>
    componentLinearVelScale=<1 1 1>
    gravityOverride {z=-980}
    physicsBodyDefinitions [{
      sphereCollisionRadius=0
      boundBone {boneName=dyn_L_sleeve_2_jnt}
      boxExtents=<10 10 20>
      constraintSetup {angularLimitsMax=<1 30 15>, angularLimitsMin=<-1 -45 -15>, angularTarget=<0 0 0>}
      localJointOffset=<2 0 0>
     }]
    relativeSpaceBone {boneName=dyn_sleeves_tie_jnt}
    sphericalLimits [{limitRadius=7, drivingBone {boneName=deform_thigh_fr_l}, sphereLocalOffset=<-8 0 0>}]
   }
   n67 = Anim Dynamics - Bone: dyn_L_sleeve_3_jnt(ComponentPose=@n66, Alpha=1, ExternalForce="0,0,0") {
    angularDampingOverride=0.9999
    angularSpringConstant=100
    bAngularSpring=true
    bOverrideAngularDamping=true
    bUseSphericalLimits=true
    simulationSpace=RootRelative
    alphaBoolBlend {blendOption=Linear}
    boundBone {boneName=dyn_L_sleeve_3_jnt}
    componentAppliedLinearAccClamp=<3500 3500 3500>
    componentLinearAccScale=<1 1 1>
    componentLinearVelScale=<1 1 1>
    gravityOverride {z=-980}
    physicsBodyDefinitions [{
      sphereCollisionRadius=0
      boundBone {boneName=dyn_L_sleeve_3_jnt}
      boxExtents=<10 10 20>
      constraintSetup {angularLimitsMax=<1 30 15>, angularLimitsMin=<-1 -45 -15>, angularTarget=<0 0 0>}
      localJointOffset=<2 0 0>
     }]
    relativeSpaceBone {boneName=dyn_sleeves_tie_jnt}
    sphericalLimits [{limitRadius=8, drivingBone {boneName=deform_thigh_fr_l}, sphereLocalOffset=<-10 0 0>}]
   }
   n68 = Anim Dynamics - Bone: dyn_R_sleeve_1_jnt(ComponentPose=@n67, Alpha=1, ExternalForce="0,0,0") {
    angularDampingOverride=0.9999
    angularSpringConstant=100
    bAngularSpring=true
    bOverrideAngularDamping=true
    bUseSphericalLimits=true
    simulationSpace=RootRelative
    alphaBoolBlend {blendOption=Linear}
    boundBone {boneName=dyn_R_sleeve_1_jnt}
    componentAppliedLinearAccClamp=<3500 3500 3500>
    componentLinearAccScale=<1 1 1>
    componentLinearVelScale=<1 1 1>
    gravityOverride {z=-980}
    physicsBodyDefinitions [{
      sphereCollisionRadius=0
      boundBone {boneName=dyn_R_sleeve_1_jnt}
      boxExtents=<10 10 20>
      constraintSetup {angularLimitsMax=<1 30 15>, angularLimitsMin=<-1 -45 -15>, angularTarget=<0 0 0>}
      localJointOffset=<2 0 0>
     }]
    relativeSpaceBone {boneName=dyn_sleeves_tie_jnt}
    sphericalLimits [{limitRadius=7, drivingBone {boneName=pelvis}, sphereLocalOffset=<0 0 0>}]
   }
   n69 = Anim Dynamics - Bone: dyn_R_sleeve_2_jnt(ComponentPose=@n68, Alpha=1, ExternalForce="0,0,0") {
    angularDampingOverride=0.9999
    angularSpringConstant=100
    bAngularSpring=true
    bOverrideAngularDamping=true
    bUseSphericalLimits=true
    simulationSpace=RootRelative
    alphaBoolBlend {blendOption=Linear}
    boundBone {boneName=dyn_R_sleeve_2_jnt}
    componentAppliedLinearAccClamp=<3500 3500 3500>
    componentLinearAccScale=<1 1 1>
    componentLinearVelScale=<1 1 1>
    gravityOverride {z=-980}
    physicsBodyDefinitions [{
      sphereCollisionRadius=0
      boundBone {boneName=dyn_R_sleeve_2_jnt}
      boxExtents=<10 10 20>
      constraintSetup {angularLimitsMax=<1 30 15>, angularLimitsMin=<-1 -45 -15>, angularTarget=<0 0 0>}
      localJointOffset=<2 0 0>
     }]
    relativeSpaceBone {boneName=dyn_sleeves_tie_jnt}
    sphericalLimits [{limitRadius=7, drivingBone {boneName=pelvis}, sphereLocalOffset=<-8 0 0>}]
   }
   n70 = "Transform (Modify) Bone - Bone: dyn_sleeves_tie_jnt"(ComponentPose=@n69, Alpha=1, Translation="-3.5,0,-6", Rotation="-24,43,10") {
    alpha=0
    rotationMode=BMM_Replace
    rotationSpace=BCS_BoneSpace
    scaleMode=BMM_Additive
    scaleSpace=BCS_BoneSpace
    translationMode=BMM_Replace
    translationSpace=BCS_BoneSpace
    alphaBoolBlend {blendOption=Linear}
    alphaScaleBiasClamp {bInterpResult=true, interpSpeedDecreasing=7, interpSpeedIncreasing=5}
    boneToModify {boneName=dyn_sleeves_tie_jnt}
    rotation=<rot -24 43 10>
    translation {x=-3.5, z=-6}
   }
   n71 = Look At - Bone: dyn_sleeves_tie_jnt(ComponentPose=@n70, Alpha=Get thigh_rot_up_r()) {
    alphaBoolBlend {blendOption=Linear}
    alphaScaleBiasClamp {bInterpResult=true, interpSpeedDecreasing=5, interpSpeedIncreasing=7}
    boneToModify {boneName=dyn_sleeves_tie_jnt}
    lookAtLocation {x=120}
    lookAtTarget {boneReference {boneName=thigh_r}}
    lookAt_Axis {axis=<1 0>}
    lookUp_Axis {axis {x=0}}
   }
   n72 = "Transform (Modify) Bone - Bone: dyn_hoodie_back_1_1_jnt"!(ComponentPose=@n71, Alpha=1, Translation="0.906,-0.192,-2.030", Rotation="-5.093,0,9.981") -> (Pose) {
    alpha=0
    rotationMode=BMM_Replace
    rotationSpace=BCS_BoneSpace
    scaleMode=BMM_Additive
    scaleSpace=BCS_BoneSpace
    translationMode=BMM_Replace
    translationSpace=BCS_BoneSpace
    alphaBoolBlend {blendOption=Linear}
    alphaScaleBiasClamp {bInterpResult=true, interpSpeedDecreasing=4, interpSpeedIncreasing=7}
    boneToModify {boneName=dyn_hoodie_back_1_1_jnt}
    rotation {pitch=-5.093, roll=9.981}
    translation=<0.906 -0.192 -2.03>
   }
   n73 = Blend Poses by bool(
    BlendPose_0=Component To Local(ComponentPose=@n72)
    BlendPose_1=Blend Poses by bool(
     BlendPose_0=Component To Local(ComponentPose=@n62)
     BlendPose_1=Blend Poses by bool(BlendPose_0=@n46, BlendPose_1=@n44, BlendTime_1=0.15, bActiveValue=@n47) {blendType=Linear}
     bActiveValue=Get bIsInFrontEnd()
    ) {blendType=Linear}
    bActiveValue=Get bIsCrouching()
   ) {blendType=Linear}
   n74 = RigidBody(ComponentPose=Local To Component(LocalPose=Use cached pose 'Pose'())) {
    bClampLinearTranslationLimitToRefPose=true
    bOverrideWorldGravity=true
    overridePhysicsAsset=&F_MED_Zest_Physics
    simulationSpace=BaseBoneSpace
    alphaBoolBlend {blendOption=Linear}
    baseBoneRef {boneName=pelvis}
    componentAppliedLinearAccClamp=<3000 3000 3000>
    componentLinearAccScale=<0.3 0.2 0.2>
    componentLinearVelScale=<0.3 0.2 0.2>
    overrideWorldGravity {z=-980}
   }
   n75 = "Transform (Modify) Bone - Bone: dyn_hoodie_back_1_1_jnt"!(ComponentPose=@n74, Alpha=1, Translation="-4.223,0.016,-1.395", Rotation="-20.000,-0.003,-0.001", Scale:Vector="1,1,1") -> (Pose) {
    rotationMode=BMM_Replace
    rotationSpace=BCS_BoneSpace
    scaleMode=BMM_Additive
    scaleSpace=BCS_BoneSpace
    translationMode=BMM_Replace
    translationSpace=BCS_BoneSpace
    alphaBoolBlend {blendOption=Linear}
    boneToModify {boneName=dyn_hoodie_back_1_1_jnt}
    rotation=<rot -20 -0.003 -0.001>
    translation=<-4.223 0.016 -1.395>
   }
   n76 = "Transform (Modify) Bone - Bone: dyn_hoodie_back_1_2_jnt"(ComponentPose=@n75, Alpha=1, Translation="0,0,0", Rotation="-15,0.000,0.000", Scale="1,1,1") {
    rotationMode=BMM_Replace
    rotationSpace=BCS_BoneSpace
    scaleMode=BMM_Additive
    scaleSpace=BCS_BoneSpace
    translationMode=BMM_Replace
    translationSpace=BCS_BoneSpace
    alphaBoolBlend {blendOption=Linear}
    boneToModify {boneName=dyn_hoodie_back_1_2_jnt}
    rotation {pitch=-15}
   }
   n77 = "Transform (Modify) Bone - Bone: dyn_hoodie_back_2_1_jnt"(ComponentPose=@n76, Alpha=1, Translation="-2.530,-4.943,-0.000", Rotation="0.000,-25.000,0.000", Scale="1,1,1") {
    rotationMode=BMM_Replace
    rotationSpace=BCS_BoneSpace
    scaleMode=BMM_Additive
    scaleSpace=BCS_BoneSpace
    translationMode=BMM_Replace
    translationSpace=BCS_BoneSpace
    alphaBoolBlend {blendOption=Linear}
    boneToModify {boneName=dyn_hoodie_back_2_1_jnt}
    rotation {yaw=-24.9998}
    translation=<-2.53 -4.943>
   }
   n78 = "Transform (Modify) Bone - Bone: dyn_hoodie_back_2_2_jnt"(ComponentPose=@n77, Alpha=1, Translation="0,0,0", Rotation="0.000,-20,0.000", Scale="1,1,1") {
    rotationMode=BMM_Replace
    rotationSpace=BCS_BoneSpace
    scaleMode=BMM_Additive
    scaleSpace=BCS_BoneSpace
    translationMode=BMM_Replace
    translationSpace=BCS_BoneSpace
    alphaBoolBlend {blendOption=Linear}
    boneToModify {boneName=dyn_hoodie_back_2_2_jnt}
    rotation {yaw=-20}
   }
   n79 = "Transform (Modify) Bone - Bone: dyn_hoodie_back_3_1_jnt"(ComponentPose=@n78, Alpha=1, Translation="-0.885,-2.436,1.033", Rotation="0,-20,0", Scale="1,1,1") {
    rotationMode=BMM_Replace
    rotationSpace=BCS_BoneSpace
    scaleMode=BMM_Additive
    scaleSpace=BCS_BoneSpace
    translationMode=BMM_Replace
    translationSpace=BCS_BoneSpace
    alphaBoolBlend {blendOption=Linear}
    boneToModify {boneName=dyn_hoodie_back_3_1_jnt}
    rotation {yaw=-20}
    translation=<-0.885 -2.436 1.033>
   }
   n80 = "Transform (Modify) Bone - Bone: dyn_hoodie_back_3_2_jnt"(ComponentPose=@n79, Alpha=1, Translation="0,0,0", Rotation="0.000,-20,0.000", Scale="1,1,1") {
    rotationMode=BMM_Replace
    rotationSpace=BCS_BoneSpace
    scaleMode=BMM_Additive
    scaleSpace=BCS_BoneSpace
    translationMode=BMM_Replace
    translationSpace=BCS_BoneSpace
    alphaBoolBlend {blendOption=Linear}
    boneToModify {boneName=dyn_hoodie_back_3_2_jnt}
    rotation {yaw=-20}
   }
   n81 = "Transform (Modify) Bone - Bone: dyn_grenade1_jnt"(ComponentPose=@n80, Alpha=1, Translation="0.000,3.121,0.000", Rotation="0,0,0", Scale="1,1,1") {
    rotationMode=BMM_Replace
    rotationSpace=BCS_BoneSpace
    scaleMode=BMM_Additive
    scaleSpace=BCS_BoneSpace
    translationMode=BMM_Replace
    translationSpace=BCS_BoneSpace
    alphaBoolBlend {blendOption=Linear}
    boneToModify {boneName=dyn_grenade1_jnt}
    translation {y=3.121}
   }
   n82 = "Transform (Modify) Bone - Bone: dyn_grenade2_jnt"(ComponentPose=@n81, Alpha=1, Translation="-0.000,2.873,-0.000", Rotation="0,0,0", Scale="1,1,1") {
    rotationMode=BMM_Replace
    rotationSpace=BCS_BoneSpace
    scaleMode=BMM_Additive
    scaleSpace=BCS_BoneSpace
    translationMode=BMM_Replace
    translationSpace=BCS_BoneSpace
    alphaBoolBlend {blendOption=Linear}
    boneToModify {boneName=dyn_grenade2_jnt}
    translation {y=2.873}
   }
   n83 = "Transform (Modify) Bone - Bone: dyn_sleeves_tie_jnt"!(ComponentPose=@n82, Alpha=1, Translation="-2.563,0.769,2.137", Rotation="33.826,2.995,-22.760", Scale:Vector="1,1,1") -> (Pose) {
    rotationMode=BMM_Replace
    rotationSpace=BCS_BoneSpace
    scaleMode=BMM_Additive
    scaleSpace=BCS_BoneSpace
    translationMode=BMM_Replace
    translationSpace=BCS_BoneSpace
    alphaBoolBlend {blendOption=Linear}
    boneToModify {boneName=dyn_sleeves_tie_jnt}
    rotation=<rot 33.826 2.995 -22.7599>
    translation=<-2.5629 0.7688 2.137>
   }
   n84 = Component To Local(ComponentPose=@n83)
   n85 = Get bIsSkyDiving()
   n86 = "Transform (Modify) Bone - Bone: dyn_sleeves_tie_jnt"!(
    ComponentPose=Local To Component(LocalPose=Blend Poses by bool(BlendPose_0=@n84, BlendPose_1=@n73, bActiveValue=@n85) {blendType=Linear})
    Alpha=1
    Translation="-4.5,-2,-8"
    Rotation="-11.138,29.883,16.654"
   ) -> (Pose) {
    alpha=0
    rotationMode=BMM_Replace
    rotationSpace=BCS_BoneSpace
    scaleMode=BMM_Additive
    scaleSpace=BCS_BoneSpace
    translationMode=BMM_Replace
    translationSpace=BCS_BoneSpace
    alphaBoolBlend {blendOption=Linear}
    boneToModify {boneName=dyn_sleeves_tie_jnt}
    rotation=<rot -11.138 29.883 16.6541>
    translation=<-4.5 -2 -8>
   }
   Output Pose(Result=Component To Local(ComponentPose=@n86)) {layerGroup=None}
   n87 = Control Rig(Source=Copy Pose From Mesh() {bUseAttachedParent=true}) {
    alphaBoolBlend {blendOption=Linear}
    controlRigAssetReference {blueprintRigClass=&/F_MED_Zest_CtrlRig_Twist.F_MED_Zest_CtrlRig_Twist_C}
   }
   n88 = Control Rig(Source=@n87) {
    alphaBoolBlend {blendOption=Linear}
    controlRigAssetReference {blueprintRigClass=&/F_MED_Zest_CtrlRig_Sculpt.F_MED_Zest_CtrlRig_Sculpt_C}
   }
   Save cached pose 'Pose'(Pose=@n88) {cacheName=Pose}
   Get GravityOverride()
  ]}
 EventGraph {nodes [
   Event BlueprintUpdateAnimation()
   > n1 = Sequence!(execute) -> (then_0, then_1, then_2:Exec, then_3:Exec, then_4:Exec)
   > n2 = Gravity Override State Function DEPRECATED(CharacterPartAnimInstance=Self-Reference(), GameTimeInSeconds=GetGameTimeInSeconds(), GravityOverrideParams=Get Gravity Override Params())
   > Set GravityOverride(GravityOverride=@n2)
   n3 = Get PawnMovementDirectionAzimuth() -> (PawnMovementDirectionAzimuth:float)
   n4 = float - float(A=MapRangeClamped(Value=@n3, InRangeA=45, InRangeB=90), B=MapRangeClamped(Value=@n3, InRangeA=90, InRangeB=135)) {is=title K2Node_CallFunction}
   n5 = Get bIsCrouching()
   n6 = Get bIsMoving2D()
   Set Right_Direct(execute=@n1.then_1, Right_Direct:double=SelectFloat(A="Absolute (Float)"(A=@n4), B=0, bPickA=AND Boolean(A=@n5, B=@n6))) -> (Output_Get:double)
   n7 = Get PawnMovementDirectionAzimuth()
   n8 = float - float(A=MapRangeClamped(Value=@n7, InRangeA=-45, InRangeB=-90), B=MapRangeClamped(Value=@n7, InRangeA=-90, InRangeB=-135))
   n9 = Get bIsCrouching()
   n10 = Get bIsMoving2D()
   > Set Left_Direct(Left_Direct:double=SelectFloat(A="Absolute (Float)"(A=@n8), B=0, bPickA=AND Boolean(A=@n9, B=@n10))) -> (Output_Get:double)
   n11 = Get PawnMovementDirectionAzimuth()
   n12 = float - float(A=MapRangeClamped(Value=@n11, InRangeA=-1, InRangeB=0.1), B=MapRangeClamped(Value=@n11, InRangeA=0.2, InRangeB=1))
   n13 = Get bIsCrouching()
   n14 = Get bIsMoving2D()
   > Set Forward_Direct(Forward_Direct:double=SelectFloat(A="Absolute (Float)"(A=@n12), B=0, bPickA=AND Boolean(A=@n13, B=@n14))) -> (Output_Get:double)
   n15 = Get PawnMovementDirectionAzimuth()
   n16 = float + float(
    A=MapRangeClamped(Value=@n15, InRangeA=-135, InRangeB=-180, OutRangeB=0.75)
    B=MapRangeClamped(Value=@n15, InRangeA=135, InRangeB=180, OutRangeB=0.75)
   ) {is=title K2Node_CommutativeAssociativeBinaryOperator}
   n17 = Get bIsMoving2D()
   n18 = Get bIsCrouching()
   > Set Backward_Direct(Backward_Direct:double=SelectFloat(A="Absolute (Float)"(A=@n16), B=0, bPickA=AND Boolean(A=@n18, B=@n17))) -> (Output_Get:double)
   n19 = Get bIsMoving2D()
   n20 = Branch(execute=@n1.then_3, Condition=AND Boolean(A=Get bIsCrouching(), B=NOT Boolean(A=@n19)))
   > Set Moving(Moving:double=1) -> (Output_Get:double)
   Set Moving(execute=@n20.else, Moving=0)
   n21 = Branch(execute=@n1.then_4, Condition=Get bIsPlayingEmote())
   > Set Emote_Off(Emote_Off:double=0) -> (Output_Get:double)
   Set Emote_Off(execute=@n21.else, Emote_Off=1)
   n22 = Get bIsMoving2D()
   n23 = Branch(execute=@n1.then_2, Condition=AND Boolean(A=Get bIsCrouching(), B=@n22))
   n24 = BreakTransform(InTransform=GetDeltaTransformFromRefPose(self=GetOwnerSkeletalMeshComponent(), BoneName=thigh_r, BaseName=pelvis))
   > Set thigh_rot_up_r(thigh_rot_up_r:double=MapRangeClamped(Value=BreakRotator(InRot=@n24.Rotation).Yaw, InRangeA=-15, InRangeB=-100)) -> (Output_Get:double)
   Set thigh_rot_up_r(execute=@n23.else, thigh_rot_up_r=0)
   NOT Boolean(A=@n22)
  ]}
]
variables [
 Backward_Direct {type=number}
 Emote_Off {type=number}
 Forward_Direct {type=number}
 Gravity Override Params {type=GravityOverrideParamsStruct}
 GravityOverride {type=Vector}
 Left_Direct {type=number}
 Moving {type=number}
 Right_Direct {type=number}
 thigh_rot_up_r {type=number}
]
